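# Physics asset for the Dragonfruit medium male backpack cosmetic, simulating a dangling strap on the pack mesh.
# It has two capsule bodies: a short, wide one on the pack bone and a long, thin one on the dyn_strap bone.
# A single constraint joins the strap to the pack with tight twist and one tight swing axis, leaving wider swing on the other.
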PhysicsAsset /BRCosmetics/Accessories/FORT_Backpacks/Backpack_M_MED_Dragonfruit/Meshes/M_MED_Dragonfruit_Pack_Physics
bodies {
 dyn_strap [{shapeType=Capsule, length=50, radius=3, center=<-3.0865 0 0>, rotation=<rot 0 90 90>}]
 pack [{shapeType=Capsule, length=8, radius=10, center=<0 -3.2258 -5.2986>}]
}
constraints [{bone1Name=dyn_strap, bone2Name=pack, swing1LimitDegrees=5, swing2LimitDegrees=45, twistLimitDegrees=10}]
