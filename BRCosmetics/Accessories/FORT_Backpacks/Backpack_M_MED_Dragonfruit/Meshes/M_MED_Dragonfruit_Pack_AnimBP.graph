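# Animation Blueprint for the Dragonfruit backpack cosmetic, adding physics motion and state-based offsets to the pack.
# A RigidBody node simulates in the pack bone's space with custom gravity and is switched off in the front end.
# Chained additive Transform Bone nodes then nudge the pack in the front end and pitch it, with interpolated blending, for skydive dive and float.
# The update event converts front-end and skydive booleans into float alphas driving these nodes.
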AnimBlueprint /BRCosmetics/Accessories/FORT_Backpacks/Backpack_M_MED_Dragonfruit/Meshes/M_MED_Dragonfruit_Pack_AnimBP
target_skeleton=&/M_MED_Dragonfruit_Pack_Skeleton
graphs [
 AnimGraph {nodes [
   n1 = RigidBody(ComponentPose=Local To Component(LocalPose=Input Pose()), Alpha=Get isNOTfrontEnd() -> (isNOTfrontEnd:double)) {
    bClampLinearTranslationLimitToRefPose=true
    bOverrideWorldGravity=true
    overridePhysicsAsset=&/M_MED_Dragonfruit_Pack_Physics
    simulationSpace=BaseBoneSpace
    alphaBoolBlend {blendOption=Linear}
    baseBoneRef {boneName=pack}
    componentLinearAccScale=<0.3 0.3 0.3>
    componentLinearVelScale=<0.4 0.4 0.4>
    overrideWorldGravity {z=-800}
   }
   n2 = "Transform (Modify) Bone - Bone: pack"(ComponentPose=@n1, Alpha=Get isFrontEndFloat() -> (isFrontEndFloat:double), Translation="0,-1.5,0", Rotation="0,0,0") {translationMode=BMM_Additive, alphaBoolBlend {blendOption=Linear}, boneToModify {boneName=pack}, translation {y=-1.5}}
   n3 = "Transform (Modify) Bone - Bone: pack"(ComponentPose=@n2, Alpha=Get Is Skydive Dive Mode() -> (Is Skydive Dive Mode:double), Translation="0,0,0", Rotation="15,0,0") {
    rotationMode=BMM_Additive
    translationMode=BMM_Additive
    alphaBoolBlend {blendOption=Linear}
    alphaScaleBiasClamp {bInterpResult=true, interpSpeedDecreasing=20, interpSpeedIncreasing=20}
    boneToModify {boneName=pack}
    rotation {pitch=15}
   }
   n4 = "Transform (Modify) Bone - Bone: pack"(ComponentPose=@n3, Alpha=Get Is Skydive Floating() -> (Is Skydive Floating:double), Translation="0,0,0", Rotation="-3,0,0") {
    rotationMode=BMM_Additive
    translationMode=BMM_Additive
    alphaBoolBlend {blendOption=Linear}
    alphaScaleBiasClamp {bInterpResult=true, interpSpeedDecreasing=20, interpSpeedIncreasing=20}
    boneToModify {boneName=pack}
    rotation {pitch=-3}
   }
   Output Pose(Result=Component To Local(ComponentPose=@n4)) {layerGroup=None}
  ]}
 EventGraph {nodes [
   Event BlueprintUpdateAnimation()
   > n1 = Sequence!(execute) -> (then_0, then_1, then_2:Exec, then_3:Exec)
   > Set Is Skydive Dive Mode(Is Skydive Dive Mode:double="To Float (Boolean)"(InBool=Get bIsSkydiveDiveMode())) -> (Output_Get:double)
   Set Is Skydive Floating(execute=@n1.then_1, Is Skydive Floating:double="To Float (Boolean)"(InBool=Get bIsSkydiveFloating())) -> (Output_Get:double)
   n2 = Branch(execute=@n1.then_2, Condition=Get bIsInFrontEnd())
   > Set isFrontEndFloat(isFrontEndFloat:double=1) -> (Output_Get:double)
   Set isFrontEndFloat(execute=@n2.else, isFrontEndFloat=0)
   n3 = Branch(execute=@n1.then_3, Condition=Get bIsInFrontEnd())
   > Set isNOTfrontEnd(isNOTfrontEnd:double=0) -> (Output_Get:double)
   Set isNOTfrontEnd(execute=@n3.else, isNOTfrontEnd=1)
   TryGetPawnOwner()
  ]}
]
variables [
 Is Skydive Dive Mode {type=number}
 Is Skydive Floating {type=number}
 isFrontEndFloat {type=number}
 isNOTfrontEnd {type=number}
]
